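# Physics asset for the male medium ModernMilitary Crisp body cosmetic, covering its left coat flaps and front, back and right skirt panels.
# Capsules shape the spine, pelvis and left leg only, with boxes for the cloth bones.
# Tight angular limits chain three coat flaps from the left thigh and join each skirt panel to its base.
# Angularly free constraints link the bases to thigh or pelvis, and two flaps to front and back panels.
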PhysicsAsset /BRCosmetics/Characters/Player/Male/Medium/Bodies/M_MED_ModernMilitary_Crisp/Meshes/M_MED_ModernMilitary_Crisp_Physics
bodies {
 calf_l [{shapeType=Capsule, length=43.3949, radius=7.8595, center=<-22.83 -0.5039 0.6139>, rotation=<rot 0 87.5936 -89.3359>}]
 coat_flap1_l [{shapeType=Box, extentX=13.9667, extentY=1.3015, extentZ=23.3973, center=<-6.0512 0.0001 -1.1496>, rotation=<rot 5 0 79.9994>}]
 coat_flap2_l [{shapeType=Box, extentX=9.1477, extentY=1.3015, extentZ=26.5047, center=<-4.8315 0.0294 -0.3624>, rotation=<rot 5 0 69.998>}]
 coat_flap3_l [{shapeType=Box, extentX=15.174, extentY=1.3015, extentZ=25.4433, center=<-5 0.3043 0.4993>, rotation=<rot 0 0 69.998>}]
 dyn_skirtBack [{shapeType=Box, extentX=14.9264, extentY=2.2445, extentZ=13.5789, center=<7 0 0>}]
 dyn_skirtBack_Base [{shapeType=Box, extentX=2.571, extentY=2.2445, extentZ=2.3207}]
 dyn_skirtFront [{shapeType=Box, extentX=14.9264, extentY=3.7325, extentZ=13.5789, center=<7 -1 0>, rotation=<rot 0.0001 -0.0001 10>}]
 dyn_skirtFront_Base [{shapeType=Box, extentX=2.571, extentY=2.2445, extentZ=2.3207}]
 dyn_skirtRight [{shapeType=Box, extentX=14.9264, extentY=2.2445, extentZ=13.5789, center=<7 0 0>}]
 dyn_skirtRight_Base [{shapeType=Box, extentX=2.571, extentY=2.2445, extentZ=2.3207}]
 pelvis [{shapeType=Capsule, length=12.3957, radius=10.5103}]
 spine_01 [{shapeType=Capsule, length=0, radius=14.75, center=<7.1334 -0.0002 0>}]
 thigh_l [
  {shapeType=Capsule, length=26.0145, radius=9.3118, center=<-18.0702 -0.0003 0>, rotation=<rot -90 -0.0074 0.0074>}
  {shapeType=Capsule, length=5.25, radius=10.5, center=<-17.5102 0.5933 -0.3279>, rotation=<rot -90 0.0198 -0.0199>}
 ]
}
constraints [
 {bone1Name=dyn_skirtRight, bone2Name=dyn_skirtRight_Base, swing1LimitDegrees=25, swing2LimitDegrees=12, twistLimitDegrees=6}
 {
  bone1Name=thigh_l
  bone2Name=pelvis
  swing1LimitDegrees=45
  swing1Motion=Free
  swing2LimitDegrees=45
  swing2Motion=Free
  twistLimitDegrees=45
  twistMotion=Free
 }
 {
  bone1Name=spine_01
  bone2Name=pelvis
  swing1LimitDegrees=45
  swing1Motion=Free
  swing2LimitDegrees=45
  swing2Motion=Free
  twistLimitDegrees=45
  twistMotion=Free
 }
 {
  bone1Name=dyn_skirtRight_Base
  bone2Name=pelvis
  swing1LimitDegrees=45
  swing1Motion=Free
  swing2LimitDegrees=45
  swing2Motion=Free
  twistLimitDegrees=45
  twistMotion=Free
 }
 {
  bone1Name=calf_l
  bone2Name=thigh_l
  swing1LimitDegrees=45
  swing1Motion=Free
  swing2LimitDegrees=45
  swing2Motion=Free
  twistLimitDegrees=45
  twistMotion=Free
 }
 {bone1Name=coat_flap1_l, bone2Name=thigh_l, swing1LimitDegrees=12, swing2LimitDegrees=12, twistLimitDegrees=4}
 {
  bone1Name=dyn_skirtBack_Base
  bone2Name=thigh_l
  swing1LimitDegrees=45
  swing1Motion=Free
  swing2LimitDegrees=45
  swing2Motion=Free
  twistLimitDegrees=45
  twistMotion=Free
 }
 {
  bone1Name=dyn_skirtFront_Base
  bone2Name=thigh_l
  swing1LimitDegrees=45
  swing1Motion=Free
  swing2LimitDegrees=45
  swing2Motion=Free
  twistLimitDegrees=45
  twistMotion=Free
 }
 {bone1Name=coat_flap2_l, bone2Name=coat_flap1_l, swing1LimitDegrees=12, swing2LimitDegrees=12, twistLimitDegrees=4}
 {
  bone1Name=coat_flap1_l
  bone2Name=dyn_skirtFront
  swing1LimitDegrees=45
  swing1Motion=Free
  swing2LimitDegrees=45
  swing2Motion=Free
  twistLimitDegrees=45
  twistMotion=Free
 }
 {bone1Name=coat_flap3_l, bone2Name=coat_flap2_l, swing1LimitDegrees=12, swing2LimitDegrees=12, twistLimitDegrees=4}
 {
  bone1Name=coat_flap2_l
  bone2Name=dyn_skirtBack
  swing1LimitDegrees=45
  swing1Motion=Free
  swing2LimitDegrees=45
  swing2Motion=Free
  twistLimitDegrees=45
  twistMotion=Free
 }
 {bone1Name=dyn_skirtBack, bone2Name=dyn_skirtBack_Base, swing1LimitDegrees=20, swing2LimitDegrees=8, twistLimitDegrees=4}
 {bone1Name=dyn_skirtFront, bone2Name=dyn_skirtFront_Base, swing1LimitDegrees=20, swing2LimitDegrees=8, twistLimitDegrees=4}
]
