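# Post-process style Anim Blueprint for the ModernMilitary Crisp medium male body, layering rig correction and secondary physics onto the pose copied from its attached parent mesh.
# A Control Rig adjusts the copied pose using pistol, rifle and emote alphas, then a partially blended RigidBody node simulates in pelvis space.
# Each update derives the weapon alphas from the owner's FortPlayerAnimInstance weapon type, grouping shotguns, snipers and rocket launchers as rifles.
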AnimBlueprint /BRCosmetics/Characters/Player/Male/Medium/Bodies/M_MED_ModernMilitary_Crisp/Meshes/M_MED_ModernMilitary_Crisp_AnimBP
target_skeleton=&/M_MED_ModernMilitary_Crisp_Skeleton
graphs [
 AnimGraph {nodes [
   n1 = RigidBody(ComponentPose=Local To Component(LocalPose=Use cached pose 'PreScale Cache'()), Alpha=0.6) {
    bClampLinearTranslationLimitToRefPose=true
    bOverrideWorldGravity=true
    overridePhysicsAsset=&/M_MED_ModernMilitary_Crisp_Physics
    simulationSpace=BaseBoneSpace
    alphaBoolBlend {blendOption=Linear}
    baseBoneRef {boneName=pelvis}
    componentAppliedLinearAccClamp=<2000 2000 4000>
    componentLinearAccScale=<0.5 0.3 0.5>
    componentLinearVelScale=<0.3 0.1 0.3>
    overrideWorldGravity {z=-980}
   }
   Output Pose(Result=Component To Local(ComponentPose=@n1)) {layerGroup=None}
   n2 = Control Rig!(
    Source=Copy Pose From Mesh() {bUseAttachedParent=true}
    Alpha
    Is Pistol Alpha:double=Get Is Pistol Alpha() -> (Is Pistol Alpha:double)
    Is Rifle Alpha:double=Get Is Rifle Alpha() -> (Is Rifle Alpha:double)
    Is Emote Alpha:double=Get Is Emote Alpha() -> (Is Emote Alpha:double)
   ) -> (Pose) {
    alphaBoolBlend {blendOption=Linear}
    controlRigAssetReference {blueprintRigClass=&/M_MED_ModernMilitary_Crisp_CtrlRig.M_MED_ModernMilitary_Crisp_CtrlRig_C}
   }
   Save cached pose 'PreScale Cache'(Pose=@n2) {cacheName=PreScale Cache}
  ]}
 EventGraph {nodes [
   Event BlueprintUpdateAnimation()
   > n1 = Sequence!(execute) -> (then_0, then_1, then_2:Exec)
   n2 = Sequence!(execute=@n1.then_1) -> (then_0, then_1, then_2, then_3:Exec, then_4:Exec)
   > n3 = Cast To FortPlayerAnimInstance(Object=GetAnimInstance(self=GetOwnerSkeletalMeshComponent()))
   > Set isPistol(isPistol=false)
   > Set isRifle(isRifle:bool=false) -> (Output_Get:bool)
   > n4 = Switch on EFortWeaponCoreAnimation!(execute, Selection:EFortWeaponCoreAnimation Enum=Get WeaponCoreAnim(self=@n3)) -> (
    Melee:Exec
    Pistol:Exec
    Shotgun:Exec
    PaperBlueprint:Exec
    Rifle:Exec
    MeleeOneHand:Exec
    MachinePistol:Exec
    RocketLauncher:Exec
    GrenadeLauncher:Exec
    GoingCommando:Exec
    AssaultRifle:Exec
    TacticalShotgun:Exec
    SniperRifle:Exec
    TrapPlacement:Exec
    ShoulderLauncher:Exec
    AbilityDecoTool:Exec
    Crossbow:Exec
    C4:Exec
    RemoteControl:Exec
    DualWield:Exec
    AR_BullPup:Exec
    AR_ForwardGrip:Exec
    MedPackPaddles:Exec
    SMG_P90:Exec
    AR_DrumGun:Exec
    Consumable_Small:Exec
    Consumable_Large:Exec
    Balloon:Exec
    MountedTurret:Exec
    CreativeTool:Exec
    ExplosiveBow:Exec
    AshtonIndigo:Exec
    AshtonChicago:Exec
    MeleeDualWield:Exec
    Unarmed:Exec
   )
   Set isRifle(execute=[@n4.AssaultRifle, @n4.Rifle, @n4.RocketLauncher, @n4.Shotgun, @n4.SniperRifle, @n4.TacticalShotgun], isRifle=true)
   Set isPistol(execute=@n4.Pistol, isPistol=true)
   Set Is Rifle Alpha(execute=@n2.then_1, Is Rifle Alpha:double="To Float (Boolean)"(InBool=Get isRifle() -> (isRifle:bool))) -> (Output_Get:double)
   Set Is Pistol Alpha(execute=@n2.then_2, Is Pistol Alpha:double="To Float (Boolean)"(InBool=Get isPistol() -> (isPistol:bool))) -> (Output_Get:double)
   Set Is Emote Alpha(execute=@n1.then_2, Is Emote Alpha:double="To Float (Boolean)"(InBool=Get bIsPlayingEmote())) -> (Output_Get:double)
  ]}
]
variables [
 Is Emote Alpha {type=number}
 Is Pistol Alpha {type=number}
 Is Rifle Alpha {type=number}
 isPistol {type=boolean}
 isRifle {type=boolean}
]
